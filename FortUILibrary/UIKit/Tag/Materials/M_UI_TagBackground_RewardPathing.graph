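# UI material that draws an animated, translucent tag background, named for reward pathing.
# A padded, skewed rounded-rectangle SDF supplies the antialiased shape mask, optional outline and edge falloffs.
# Rotating eased gradients blend colors, while two counter-rotating panned noise samples multiply into drifting blob highlights.
# Named reroutes link these layers, with a time-driven ping-pong varying highlight falloff and a color dodge blend adding intensity.
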Material /FortUILibrary/UIKit/Tag/Materials/M_UI_TagBackground_RewardPathing
graph {
 properties {
  bUsedWithStaticMesh=false
  blendMode=BLEND_Translucent
  materialDomain=MD_UI
  pixelDepthOffsetMode=PDOM_Legacy
  usedShadingModels=Default Lit
 }
 nodes [
  n1 = NamedRerouteUsage()
  n2 = Multiply(A=Time()) {constB=0.2}
  n3 = Power(
   Base=MaterialFunctionCall("Value (0-1)"=SmoothStep(Value=@n1) {constMin=0.35}) {materialFunction=&MF_UI_EaseCurves}.EaseInSine
   Exp=LinearInterpolate(Alpha=MaterialFunctionCall!(Value=@n2, Scale, Height) {materialFunction=/FortUILibrary/MaterialFunctions/Basics/MF_UI_PingPong}.Result) {constA=3.5, constB=5}
  ) {constExponent=10}
  LinearInterpolate(A=NamedRerouteUsage(), Alpha=LinearInterpolate(A=NamedRerouteUsage(), Alpha=Multiply(A=@n3) {constB=0.4}))
  MaterialFunctionCall!(Blend=Min(A=NamedRerouteUsage()) {constB=0.9999}, Base=NamedRerouteUsage()) {materialFunction=/Engine/Functions/Engine_MaterialFunctions03/Blends/Blend_ColorDodge}
  n4 = OneMinus(None=MaterialFunctionCall!(Value in Pixels=ScalarParameter() {defaultValue=150, parameterName="Padding (PX)"}, Pixel Reference=NamedRerouteUsage()) {materialFunction=/FortUILibrary/MaterialFunctions/Basics/MF_UI_PixelsToNormalizedV4}."Per‑Axis Normalized (V4)")
  n5 = ScalarParameter() {defaultValue=0.85, group=Layout, parameterName=ScaleX, sortPriority=1}
  n6 = ScalarParameter() {defaultValue=1, group=Layout, parameterName=ScaleY, sortPriority=1}
  n7 = ScalarParameter() {parameterName=hasOutline}
  n8 = MaterialFunctionCall(
   Size=Multiply(
    A=ComponentMask(None=@n4) {g=true, r=true}
    B=MaterialFunctionCall(X=@n5, Y=@n6) {materialFunction=&MakeFloat2}.Result
   )
   Corner Radius=MaterialFunctionCall(
    X=ScalarParameter() {defaultValue=0.1875, group=Layout, parameterName=TopLeft, sortPriority=10}
    Y=ScalarParameter() {defaultValue=0.25, group=Layout, parameterName=BottomLeft, sortPriority=13}
    Z=ScalarParameter() {defaultValue=0.25, group=Layout, parameterName=TopRight, sortPriority=11}
    A=ScalarParameter() {defaultValue=0.1875, group=Layout, parameterName=BottomRight, sortPriority=12}
   ) {materialFunction=&MakeFloat4}.Result
   Radius in Pixels=ScalarParameter() {group=Layout, parameterName=Radius in Pixels, sortPriority=9}
   Skew Angle=ScalarParameter() {defaultValue=10, group=Layout, parameterName=SkewAngle, sortPriority=4}
   Outline Thickness=Multiply(A=ScalarParameter() {defaultValue=4, parameterName=OutlineThickness}, B=Floor(None=@n7))
   Pixel Size=NamedRerouteUsage()
  ) {materialFunction=&MF_UI_CoreUI_RoundedMask}
  n9 = OneMinus(None=@n8.SDF)
  NamedRerouteDeclaration(None=SmoothStep(Value=@n9) {constMax=2, constMin=1}) {name=SDF, nodeColor=<rgba 0 0 0 1>}
  n10 = Rotator(Coordinate=MaterialFunctionCall() {materialFunction=&MF_UI_UVScaling}.Scaled To Fill) {speed=2}
  NamedRerouteDeclaration(None=MaterialFunctionCall("Value (0-1)"=MaterialFunctionCall(UV Channel=@n10) {materialFunction=&LinearGradient}.UGradient) {materialFunction=&MF_UI_EaseCurves}.EaseInOutCubic) {name=Gradient, nodeColor=<rgba 0 0 0 1>}
  n11 = SmoothStep(Value=OneMinus(None=NamedRerouteUsage())) {constMax=1.5, constMin=-2}
  n12 = Power(Base=Multiply(A=NamedRerouteUsage(), B=Multiply(A=@n11, B=@n11)))
  n13 = NamedRerouteUsage()
  n14 = NamedRerouteUsage()
  NamedRerouteDeclaration(None=LinearInterpolate(A=Multiply(A=@n12) {constB=0.85}, Alpha=Multiply(A=@n13, B=@n14))) {name=Highlights, nodeColor=<rgba 0 0 0 1>}
  n15 = Add(A=@n8.SDF) {constB=-0.01}
  NamedRerouteDeclaration(None=Multiply(
    A=MaterialFunctionCall(DistanceField=@n15) {materialFunction=&MF_UI_AntiAliasing}.Inner
    B=ScalarParameter() {defaultValue=1, parameterName=Opacity, sliderMax=1}
   )) {name=Shape Mask, nodeColor=<rgba 0 0 0 0>}
  n16 = Constant() {r=1}
  NamedRerouteDeclaration(None=Multiply(A=Distance(A=@n9, B=@n16)) {constB=2}) {name=Outline SDF, nodeColor=<rgba 0 0 0 1>}
  n17 = NamedRerouteUsage()
  n18 = NamedRerouteUsage()
  n19 = LinearInterpolate(A=Power(Base=@n17) {constExponent=0.1}, B=@n17, Alpha=Power(Base=OneMinus(None=@n18)) {constExponent=0.1})
  n20 = NamedRerouteUsage()
  n21 = Constant3Vector() {constant=<rgba 0.4896 0.8139 1 1>}
  n22 = NamedRerouteUsage()
  n23 = LinearInterpolate(
   A=NamedRerouteDeclaration(None=@n21) {name=BaseColor, nodeColor=<rgba 0 0 0 1>}
   B=Power(Base=@n21) {constExponent=0.5}
   Alpha=Power(Base=@n22) {constExponent=3}
  )
  NamedRerouteDeclaration(None=LinearInterpolate(A=Power(Base=@n19), B=LinearInterpolate(A=@n20, B=@n23, Alpha=@n22) {constA=1}, Alpha=NamedRerouteUsage())) {name=Gradient Color, nodeColor=<rgba 0 0 0 1>}
  NamedRerouteDeclaration(None=MaterialFunctionCall!(Anchor, Translation, Padding=ScalarParameter() {defaultValue=150, parameterName="Padding (PX)"}, Scale) {materialFunction=/FortUILibrary/MaterialFunctions/Basics/MF_UI_VertexTransform}.New Pixel Size) {name=New Pixel Size, nodeColor=<rgba 0 0 0 1>}
  NamedRerouteDeclaration(None=@n8.Shape Mask) {name=Fill, nodeColor=<rgba 0.125 1 0 1>}
  n24 = MaterialFunctionCall(UV Channel=Rotator(Coordinate=MaterialFunctionCall() {materialFunction=&MF_UI_UVScaling}.Scaled To Fill) {speed=2.5}) {materialFunction=&LinearGradient}
  NamedRerouteDeclaration(None=LinearInterpolate(
    A=VectorParameter() {group=0 - Coloring, parameterName=Color 1, defaultValue=<rgba 0.162 0.4179 1 1>}.RGB
    B=VectorParameter() {group=0 - Coloring, parameterName=Color 2, defaultValue=<rgba 0.6883 0.3193 1 1>}.RGB
    Alpha=MaterialFunctionCall("Value (0-1)"=@n24.UGradient) {materialFunction=&MF_UI_EaseCurves}.EaseInOutCubic
   )) {name=Rift Gradient, nodeColor=<rgba 0 0 0 1>}
  NamedRerouteDeclaration(None=@n9) {name=Outer Glow SDF, nodeColor=<rgba 0 0 0 1>}
  n25 = MaterialFunctionCall(
   UVs=MaterialFunctionCall() {materialFunction=&MF_UI_UVScaling}.Scaled To Fill
   Texture Scale=Constant() {r=4}
  ) {materialFunction=&ScaleUVsByCenter}
  n26 = TextureSample(UVs=Panner(Coordinate=Rotator(Coordinate=@n25.UVs) {speed=0.2}) {speedX=0.2}) {samplerType=SAMPLERTYPE_Grayscale, texture=&LowResBlurredNoise}
  n27 = MaterialFunctionCall(
   UVs=MaterialFunctionCall() {materialFunction=&MF_UI_UVScaling}.Scaled To Fill
   Texture Scale=Constant() {r=5}
  ) {materialFunction=&ScaleUVsByCenter}
  n28 = TextureSample(UVs=Panner(Coordinate=Rotator(Coordinate=@n27.UVs) {speed=-0.25}) {speedX=-0.28}) {samplerType=SAMPLERTYPE_Grayscale, texture=&LowResBlurredNoise}
  n29 = Multiply(A=Power(Base=Multiply(A=@n26.R, B=@n28.R)), B=SmoothStep(Value=NamedRerouteUsage()) {constMin=0.6})
  NamedRerouteDeclaration(None=Saturate(None=Multiply(A=@n29) {constB=6})) {name=BlobsHighlights, nodeColor=<rgba 0 0 0 1>}
  NamedRerouteUsage()
 ]
}
