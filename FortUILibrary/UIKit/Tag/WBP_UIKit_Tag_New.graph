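# UIKit tag widget that displays a static "NEW" label over a slanted background.
# A root grid layers a padded, vertically centered inner grid above an image that uses the RewardPathing tag material and a horizontal render shear.
# The inner grid holds dark bold italic CommonText that auto-collapses when empty; despite its name, it contains no icon.
# The event graph has only unconnected stub events and no variables.
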WidgetBlueprint /FortUILibrary/UIKit/Tag/WBP_UIKit_Tag_New
graphs [EventGraph {nodes [
   Event PreConstruct()
   Event Construct()
   Event Tick()
  ]}]
variables []
widgets {
 info {inheritedWidgetCount=0, parentClass=&UserWidget, rootWidgetClass=&GridPanel, widgetCount=4}
 widgets [
  TagContainer {widgetClassPath=&GridPanel, bIsVariable=true, widget=TagContainer, properties {columnFill [1], rowFill [1]}}
  Background {
   widgetClassPath=&Image
   bIsVariable=true
   parent=TagContainer
   widget=Background
   properties {
    visibility=SelfHitTestInvisible
    brush {
     bIsDynamicallyLoaded=false
     drawAs=Image
     imageType=FullColor
     mirroring=NoMirror
     resourceName=None
     resourceObject=/FortUILibrary/UIKit/Tag/Materials/MI_UI_TagBackground_RewardPathing
     tiling=NoTile
     imageSize=<0 0>
     margin=<ltrb 0 0 0 0>
     outlineSettings {
      bUseBrushTransparency=false
      roundingType=HalfHeightRadius
      width=0
      color {colorUseRule=UseColor_Specified, specifiedColor=<rgba 0 0 0 0>}
      cornerRadii=<0 0 0 0>
     }
     tintColor {colorUseRule=UseColor_Specified, specifiedColor=<rgba 1 1 1 1>}
     uVRegion {bIsValid=false, max=<0 0>, min=<0 0>}
    }
    renderTransform {angle=0, scale=<1 1>, shear=<-16 0>, translation=<0 0>}
   }
   slot {class=GridSlot}
  }
  GridText_AndOptionalIcon {
   widgetClassPath=&GridPanel
   bIsVariable=true
   parent=TagContainer
   widget=GridText_AndOptionalIcon
   properties {columnFill [1,0]}
   slot {class=GridSlot, properties {layer=1, verticalAlignment=VAlign_Center, padding=<ltrb 12 0 12 0>}}
  }
  Text_Tag {
   widgetClassPath=&CommonTextBlock
   bIsVariable=true
   parent=GridText_AndOptionalIcon
   widget=Text_Tag
   properties {
    applyLineHeightToBottomLine=false
    bAutoCollapseWithEmptyText=true
    mobileFontSizeMultiplier=1.2
    pixelSnapping=Disabled
    text=NEW
    colorAndOpacity {colorUseRule=UseColor_Specified, specifiedColor=<rgba 0.0048 0.004 0.0116 1>}
    font {
     bForceMonospaced=false
     bMaterialIsStencil=false
     fontMaterial=null
     fontObject=&HeadingNow
     letterSpacing=10
     monospacedWidth=1
     size=15.75
     skewAmount=0
     typefaceFontName=86 Bold Italic
     outlineSettings {
      bApplyOutlineToDropShadows=false
      bMiteredCorners=false
      bSeparateFillAlpha=false
      outlineMaterial=null
      outlineSize=0
      outlineColor=<rgba 0 0 0 1>
     }
    }
    shadowOffset=<0 0>
   }
   slot {class=GridSlot, properties {columnSpan=1}}
  }
 ]
}
